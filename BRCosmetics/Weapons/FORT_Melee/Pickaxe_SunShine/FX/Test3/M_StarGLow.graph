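# Translucent, unlit, two-sided Niagara sprite material for a star glow effect on the SunShine pickaxe.
# The red channel of a star texture parameter is multiplied by particle alpha to form a soft mask.
# Default and Deferred shading paths apply a depth fade to that mask, while Forward and Mobile skip it.
# The switched result is multiplied by a view-angle fade with adjustable sensitivity.
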Material /BRCosmetics/Weapons/FORT_Melee/Pickaxe_SunShine/FX/Test3/M_StarGLow
graph {
 properties {
  bUsedWithNiagaraSprites=true
  blendMode=BLEND_Translucent
  pixelDepthOffsetMode=PDOM_Legacy
  shadingModel=MSM_Unlit
  twoSided=true
  usedShadingModels=Unlit
 }
 nodes [
  n1 = Multiply(
   A=TextureSampleParameter2D() {parameterName=Texturee, texture=&/T_StarFront}.R
   B=ParticleColor().A
  )
  n2 = DepthFade(Opacity=@n1, FadeDistance=ScalarParameter() {defaultValue=20, parameterName=FadeDistance})
  Multiply(
   A=ShadingPathSwitch(Default=@n2, Deferred=@n2, Forward=@n1, Mobile=@n1)
   B=MaterialFunctionCall("FadeSensitivity (optional)"=ScalarParameter() {defaultValue=2, parameterName="FadeSensitivity (optional) (S)"}) {materialFunction=&FadeBasedOnViewAngle}.Result
  )
 ]
}
