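# Translucent unlit material for Niagara sprites that draws a star shape, likely for impact effects on the SunShine pickaxe.
# The green channel of an SDF star texture borrowed from item shop UI art passes through a SmoothStep with a scalar-parameter minimum, making the shape's edge threshold adjustable.
# That mask is multiplied by particle color alpha, letting the particle system fade it.
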Material /BRCosmetics/Weapons/FORT_Melee/Pickaxe_SunShine/FX/Test3/M_StarImpact
graph {
 properties {bUsedWithNiagaraSprites=true, blendMode=BLEND_Translucent, pixelDepthOffsetMode=PDOM_Legacy, shadingModel=MSM_Unlit, usedShadingModels=Unlit}
 nodes [Multiply(
   A=SmoothStep(
    Min=ScalarParameter()
    Value=TextureSample() {samplerType=SAMPLERTYPE_Alpha, texture=/Game/UI/Frontend/ItemShop/Art/InspireSpell/T_UI_SDFstar}.G
   )
   B=ParticleColor().A
  )]
}
